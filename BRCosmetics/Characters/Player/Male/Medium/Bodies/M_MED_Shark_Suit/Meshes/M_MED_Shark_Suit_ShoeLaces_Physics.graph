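# Physics asset for the shoelaces on the male medium Shark Suit player body, defining bodies and joints for their secondary motion.
# Each foot capsule anchors two chains of three box bodies, linked by constraints with narrow twist and moderate swing limits, slightly wider in one swing axis past the first link.
# Both feet connect to a pelvis capsule through constraints with free swing and twist.
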PhysicsAsset /BRCosmetics/Characters/Player/Male/Medium/Bodies/M_MED_Shark_Suit/Meshes/M_MED_Shark_Suit_ShoeLaces_Physics
bodies {
 dyn_l_shoelace_01_a [{shapeType=Box, extentX=0.669, extentY=1.5203, extentZ=2.6108, center=<1.32 0 0>, rotation=<rot 90 0 0>}]
 dyn_l_shoelace_01_b [{shapeType=Box, extentX=0.669, extentY=1.5203, extentZ=1.7865, center=<0.8759 0 0>, rotation=<rot 90 0 0>}]
 dyn_l_shoelace_01_c [{shapeType=Box, extentX=0.669, extentY=1.5203, extentZ=2.284, center=<1.1253 0 0>, rotation=<rot 90 0 0>}]
 dyn_l_shoelace_02_a [{shapeType=Box, extentX=0.669, extentY=1.5203, extentZ=2.6108, center=<1.32 0 0>, rotation=<rot 90 0 0>}]
 dyn_l_shoelace_02_b [{shapeType=Box, extentX=0.669, extentY=1.5203, extentZ=1.7865, center=<0.8759 0 0>, rotation=<rot 90 0 0>}]
 dyn_l_shoelace_02_c [{shapeType=Box, extentX=0.669, extentY=1.5203, extentZ=2.284, center=<1.1253 0 0>, rotation=<rot 90 0 0>}]
 dyn_r_shoelace_01_a [{shapeType=Box, extentX=0.669, extentY=1.5203, extentZ=2.6108, center=<1.32 0 0>, rotation=<rot 90 0 0>}]
 dyn_r_shoelace_01_b [{shapeType=Box, extentX=0.669, extentY=1.5203, extentZ=1.7865, center=<0.8759 0 0>, rotation=<rot 90 0 0>}]
 dyn_r_shoelace_01_c [{shapeType=Box, extentX=0.669, extentY=1.5203, extentZ=2.284, center=<1.1253 0 0>, rotation=<rot 90 0 0>}]
 dyn_r_shoelace_02_a [{shapeType=Box, extentX=0.669, extentY=1.5203, extentZ=2.6108, center=<1.32 0 0>, rotation=<rot 90 0 0>}]
 dyn_r_shoelace_02_b [{shapeType=Box, extentX=0.669, extentY=1.5203, extentZ=1.7865, center=<0.8759 0 0>, rotation=<rot 90 0 0>}]
 dyn_r_shoelace_02_c [{shapeType=Box, extentX=0.669, extentY=1.5203, extentZ=2.284, center=<1.1253 0 0>, rotation=<rot 90 0 0>}]
 foot_l [{shapeType=Capsule, length=14.1072, radius=3.7365, center=<-3.1629 5.4875 -0.0001>, rotation=<rot 0 12.3814 89.9999>}]
 foot_r [{shapeType=Capsule, length=13.9064, radius=3.9373, center=<3.609 -5.7381 -0.0002>, rotation=<rot 0 12.3801 89.9999>}]
 pelvis [{shapeType=Capsule, length=9.5393, radius=6.9457}]
}
constraints [
 {bone1Name=dyn_r_shoelace_02_a, bone2Name=foot_r, swing1LimitDegrees=20, swing2LimitDegrees=15, twistLimitDegrees=2}
 {bone1Name=dyn_r_shoelace_02_b, bone2Name=dyn_r_shoelace_02_a, swing1LimitDegrees=30, swing2LimitDegrees=15, twistLimitDegrees=2}
 {bone1Name=dyn_r_shoelace_02_c, bone2Name=dyn_r_shoelace_02_b, swing1LimitDegrees=30, swing2LimitDegrees=15, twistLimitDegrees=2}
 {bone1Name=dyn_r_shoelace_01_a, bone2Name=foot_r, swing1LimitDegrees=20, swing2LimitDegrees=15, twistLimitDegrees=2}
 {
  bone1Name=foot_r
  bone2Name=pelvis
  swing1LimitDegrees=45
  swing1Motion=Free
  swing2LimitDegrees=45
  swing2Motion=Free
  twistLimitDegrees=45
  twistMotion=Free
 }
 {bone1Name=dyn_r_shoelace_01_b, bone2Name=dyn_r_shoelace_01_a, swing1LimitDegrees=30, swing2LimitDegrees=15, twistLimitDegrees=2}
 {bone1Name=dyn_r_shoelace_01_c, bone2Name=dyn_r_shoelace_01_b, swing1LimitDegrees=30, swing2LimitDegrees=15, twistLimitDegrees=2}
 {bone1Name=dyn_l_shoelace_01_c, bone2Name=dyn_l_shoelace_01_b, swing1LimitDegrees=30, swing2LimitDegrees=15, twistLimitDegrees=2}
 {bone1Name=dyn_l_shoelace_01_b, bone2Name=dyn_l_shoelace_01_a, swing1LimitDegrees=30, swing2LimitDegrees=15, twistLimitDegrees=2}
 {bone1Name=dyn_l_shoelace_01_a, bone2Name=foot_l, swing1LimitDegrees=20, swing2LimitDegrees=15, twistLimitDegrees=2}
 {bone1Name=dyn_l_shoelace_02_c, bone2Name=dyn_l_shoelace_02_b, swing1LimitDegrees=30, swing2LimitDegrees=15, twistLimitDegrees=2}
 {bone1Name=dyn_l_shoelace_02_b, bone2Name=dyn_l_shoelace_02_a, swing1LimitDegrees=30, swing2LimitDegrees=15, twistLimitDegrees=2}
 {bone1Name=dyn_l_shoelace_02_a, bone2Name=foot_l, swing1LimitDegrees=20, swing2LimitDegrees=15, twistLimitDegrees=2}
 {
  bone1Name=foot_l
  bone2Name=pelvis
  swing1LimitDegrees=45
  swing1Motion=Free
  swing2LimitDegrees=45
  swing2Motion=Free
  twistLimitDegrees=45
  twistMotion=Free
 }
]
