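# Physics asset for the teeth of the male medium Shark Suit cosmetic body, covering its dynamic jaw and six tooth bones.
# Capsule bodies run from pelvis through spine and neck to head, linked by free-rotation constraints, with a jaw offset and jaw branching off neck_02.
# Each tooth hangs off the jaw offset bone with tight twist and first-swing limits and a wide second swing, giving hinge-like wobble along one axis.
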PhysicsAsset /BRCosmetics/Characters/Player/Male/Medium/Bodies/M_MED_Shark_Suit/Meshes/M_MED_Shark_Suit_Teeth_Physics
bodies {
 dyn_jaws [{shapeType=Capsule, length=1, radius=1}]
 dyn_jaws_offset [{shapeType=Capsule, length=1, radius=1}]
 dyn_teeth_01 [{shapeType=Capsule, length=1, radius=1, center=<-0.103 -0.1288 -1.1833>}]
 dyn_teeth_02 [{shapeType=Capsule, length=1, radius=1, center=<-0.0705 0.0359 -1.1921>}]
 dyn_teeth_03 [{shapeType=Capsule, length=1, radius=1, center=<-0.1757 0.1095 -1.1766>}]
 dyn_teeth_04 [{shapeType=Capsule, length=1, radius=1, center=<0 0 -1.1947>}]
 dyn_teeth_05 [{shapeType=Capsule, length=1, radius=1, center=<0.1376 0.0058 -1.1867>}]
 dyn_teeth_06 [{shapeType=Capsule, length=1, radius=1, center=<0.2077 0.0927 -1.1729>}]
 head [{shapeType=Capsule, length=1, radius=1}]
 neck_01 [{shapeType=Capsule, length=1, radius=1}]
 neck_02 [{shapeType=Capsule, length=1, radius=1}]
 pelvis [{shapeType=Capsule, length=11.766, radius=9.3234}]
 spine_01 [{shapeType=Capsule, length=1, radius=1}]
 spine_02 [{shapeType=Capsule, length=1, radius=1}]
 spine_03 [{shapeType=Capsule, length=1, radius=1}]
 spine_04 [{shapeType=Capsule, length=1, radius=1}]
 spine_05 [{shapeType=Capsule, length=1, radius=1}]
}
constraints [
 {
  bone1Name=spine_01
  bone2Name=pelvis
  swing1LimitDegrees=45
  swing1Motion=Free
  swing2LimitDegrees=45
  swing2Motion=Free
  twistLimitDegrees=45
  twistMotion=Free
 }
 {
  bone1Name=spine_02
  bone2Name=spine_01
  swing1LimitDegrees=45
  swing1Motion=Free
  swing2LimitDegrees=45
  swing2Motion=Free
  twistLimitDegrees=45
  twistMotion=Free
 }
 {
  bone1Name=spine_03
  bone2Name=spine_02
  swing1LimitDegrees=45
  swing1Motion=Free
  swing2LimitDegrees=45
  swing2Motion=Free
  twistLimitDegrees=45
  twistMotion=Free
 }
 {
  bone1Name=spine_04
  bone2Name=spine_03
  swing1LimitDegrees=45
  swing1Motion=Free
  swing2LimitDegrees=45
  swing2Motion=Free
  twistLimitDegrees=45
  twistMotion=Free
 }
 {
  bone1Name=spine_05
  bone2Name=spine_04
  swing1LimitDegrees=45
  swing1Motion=Free
  swing2LimitDegrees=45
  swing2Motion=Free
  twistLimitDegrees=45
  twistMotion=Free
 }
 {
  bone1Name=neck_01
  bone2Name=spine_05
  swing1LimitDegrees=45
  swing1Motion=Free
  swing2LimitDegrees=45
  swing2Motion=Free
  twistLimitDegrees=45
  twistMotion=Free
 }
 {
  bone1Name=neck_02
  bone2Name=neck_01
  swing1LimitDegrees=45
  swing1Motion=Free
  swing2LimitDegrees=45
  swing2Motion=Free
  twistLimitDegrees=45
  twistMotion=Free
 }
 {
  bone1Name=head
  bone2Name=neck_02
  swing1LimitDegrees=45
  swing1Motion=Free
  swing2LimitDegrees=45
  swing2Motion=Free
  twistLimitDegrees=45
  twistMotion=Free
 }
 {
  bone1Name=dyn_jaws_offset
  bone2Name=neck_02
  swing1LimitDegrees=45
  swing1Motion=Free
  swing2LimitDegrees=45
  swing2Motion=Free
  twistLimitDegrees=45
  twistMotion=Free
 }
 {
  bone1Name=dyn_jaws
  bone2Name=dyn_jaws_offset
  swing1LimitDegrees=45
  swing1Motion=Free
  swing2LimitDegrees=45
  swing2Motion=Free
  twistLimitDegrees=45
  twistMotion=Free
 }
 {bone1Name=dyn_teeth_02, bone2Name=dyn_jaws_offset, swing1LimitDegrees=2, swing2LimitDegrees=45, twistLimitDegrees=1}
 {bone1Name=dyn_teeth_01, bone2Name=dyn_jaws_offset, swing1LimitDegrees=2, swing2LimitDegrees=45, twistLimitDegrees=1}
 {bone1Name=dyn_teeth_03, bone2Name=dyn_jaws_offset, swing1LimitDegrees=2, swing2LimitDegrees=45, twistLimitDegrees=1}
 {bone1Name=dyn_teeth_06, bone2Name=dyn_jaws_offset, swing1LimitDegrees=2, swing2LimitDegrees=45, twistLimitDegrees=1}
 {bone1Name=dyn_teeth_05, bone2Name=dyn_jaws_offset, swing1LimitDegrees=2, swing2LimitDegrees=45, twistLimitDegrees=1}
 {bone1Name=dyn_teeth_04, bone2Name=dyn_jaws_offset, swing1LimitDegrees=2, swing2LimitDegrees=45, twistLimitDegrees=1}
]
