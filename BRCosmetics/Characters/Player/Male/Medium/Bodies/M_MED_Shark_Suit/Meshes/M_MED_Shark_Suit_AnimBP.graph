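# Post-pose AnimBlueprint for the male medium Shark Suit body, adding secondary motion to cosmetic parts.
# It copies the attached parent's pose and curves, then runs a deform Control Rig.
# A chain of bone constraints, RigidBody sims and Anim Dynamics drives kneepad, laces, jaws, teeth, leg suit and pouch offsets.
# Pouch Anim Dynamics nodes are disconnected from the output, and the event graph has no logic.
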AnimBlueprint /BRCosmetics/Characters/Player/Male/Medium/Bodies/M_MED_Shark_Suit/Meshes/M_MED_Shark_Suit_AnimBP
target_skeleton=&/M_MED_Shark_Suit_Skeleton
graphs [
 AnimGraph {nodes [
   n1 = Anim Dynamics - Bone: dyn_l_pouch(Alpha=1, ExternalForce="0,0,0") {
    angularDampingOverride=0.9999
    angularSpringConstant=700
    bAngularSpring=true
    bGravityOverrideInSimSpace=true
    bLinearSpring=true
    bOverrideAngularDamping=true
    bUseGravityOverride=true
    linearSpringConstant=100
    alphaBoolBlend {blendOption=Linear}
    boundBone {boneName=dyn_l_pouch}
    componentAppliedLinearAccClamp=<3000 3000 3000>
    componentLinearAccScale=<0.1 0.1 0.1>
    gravityOverride {z=-20}
    physicsBodyDefinitions [{
      sphereCollisionRadius=0
      boundBone {boneName=dyn_l_pouch}
      boxExtents=<10 10 5>
      constraintSetup {angularLimitsMax=<5 5 5>, angularLimitsMin=<-5 -5 -5>, angularTarget=<0 0 0>}
      localJointOffset=<0 0 -1>
     }]
   }
   Anim Dynamics - Bone: dyn_r_pouch(ComponentPose=@n1, Alpha=1, ExternalForce="0,0,0") {
    angularDampingOverride=0.9999
    angularSpringConstant=700
    bAngularSpring=true
    bGravityOverrideInSimSpace=true
    bLinearSpring=true
    bOverrideAngularDamping=true
    bUseGravityOverride=true
    linearSpringConstant=100
    alphaBoolBlend {blendOption=Linear}
    boundBone {boneName=dyn_r_pouch}
    componentAppliedLinearAccClamp=<3000 3000 3000>
    componentLinearAccScale=<0.1 0.1 0.1>
    gravityOverride {z=-20}
    physicsBodyDefinitions [{
      sphereCollisionRadius=0
      boundBone {boneName=dyn_r_pouch}
      boxExtents=<10 10 5>
      constraintSetup {angularLimitsMax=<5 5 5>, angularLimitsMin=<-5 -5 -5>, angularTarget=<0 0 0>}
      localJointOffset=<1 -1 0.1>
     }]
   }
   n2 = Control Rig(Source=Copy Pose From Mesh() {bCopyCurves=true, bUseAttachedParent=true}) {
    alphaBoolBlend {blendOption=Linear}
    controlRigAssetReference {blueprintRigClass=&/M_MED_SharkSuit_DeformRig_ControlRig.M_MED_SharkSuit_DeformRig_ControlRig_C}
   }
   n3 = "Constraint - dyn_kneepad (PP)"!(ComponentPose=Local To Component(LocalPose=@n2), Alpha=1, ConstraintWeights_0=0.5, ConstraintWeights_1:float=0.5) -> (Pose) {
    alphaBoolBlend {blendOption=Linear}
    boneToModify {boneName=dyn_kneepad}
    constraintSetup [{targetBone {boneName=thigh_twist_01_r}}, {targetBone {boneName=calf_r}}]
    constraintWeights [1,1]
   }
   n4 = RigidBody(ComponentPose=@n3) {
    bClampLinearTranslationLimitToRefPose=true
    bOverrideWorldGravity=true
    overridePhysicsAsset=&/M_MED_Shark_Suit_ShoeLaces_Physics
    simulationSpace=BaseBoneSpace
    alphaBoolBlend {blendOption=Linear}
    componentAppliedLinearAccClamp=<3000 3000 3000>
    componentLinearAccScale=<0.35 0.35 0.45>
    componentLinearVelScale=<0.3 0.3 0.1>
    overrideWorldGravity {z=-980}
   }
   n5 = "Constraint - dyn_jaws_offset (P)"(ComponentPose=@n4, Alpha=1, ConstraintWeights_0=1) {
    alphaBoolBlend {blendOption=Linear}
    boneToModify {boneName=dyn_jaws_offset}
    constraintSetup [{targetBone {boneName=head}}]
    constraintWeights [1]
   }
   n6 = Anim Dynamics - Bone: dyn_jaws(ComponentPose=@n5, Alpha=0.2, ExternalForce="0,0,0") {
    angularDampingOverride=0.9999
    angularSpringConstant=50
    bAngularSpring=true
    bGravityOverrideInSimSpace=true
    bLinearSpring=true
    bOverrideAngularDamping=true
    bOverrideLinearDamping=true
    gravityScale=0.05
    linearDampingOverride=0.9999
    linearSpringConstant=150
    alphaBoolBlend {blendOption=Linear}
    boundBone {boneName=dyn_jaws}
    componentLinearAccScale=<0.25 0.25 0.25>
    physicsBodyDefinitions [{
      sphereCollisionRadius=0
      boundBone {boneName=dyn_jaws}
      boxExtents=<10 20 20>
      constraintSetup {angularLimitsMax=<45 5 45>, angularLimitsMin=<-45 -5 -45>, angularTarget=<0 0 0>, linearAxesMax=<3 3 3>, linearAxesMin=<-3 -3 -3>}
      localJointOffset=<0 0 0.1>
     }]
   }
   n7 = "Constraint - dyn_teeth_01 (P)"(ComponentPose=@n6, Alpha=1, ConstraintWeights_0=1) {
    alphaBoolBlend {blendOption=Linear}
    boneToModify {boneName=dyn_teeth_01}
    constraintSetup [{targetBone {boneName=dyn_jaws}}]
    constraintWeights [1]
   }
   n8 = "Constraint - dyn_teeth_02 (P)"(ComponentPose=@n7, Alpha=1, ConstraintWeights_0=1) {
    alphaBoolBlend {blendOption=Linear}
    boneToModify {boneName=dyn_teeth_02}
    constraintSetup [{targetBone {boneName=dyn_jaws}}]
    constraintWeights [1]
   }
   n9 = "Constraint - dyn_teeth_03 (P)"(ComponentPose=@n8, Alpha=1, ConstraintWeights_0=1) {
    alphaBoolBlend {blendOption=Linear}
    boneToModify {boneName=dyn_teeth_03}
    constraintSetup [{targetBone {boneName=dyn_jaws}}]
    constraintWeights [1]
   }
   n10 = "Constraint - dyn_teeth_04 (P)"(ComponentPose=@n9, Alpha=1, ConstraintWeights_0=1) {
    alphaBoolBlend {blendOption=Linear}
    boneToModify {boneName=dyn_teeth_04}
    constraintSetup [{targetBone {boneName=dyn_jaws}}]
    constraintWeights [1]
   }
   n11 = "Constraint - dyn_teeth_05 (P)"(ComponentPose=@n10, Alpha=1, ConstraintWeights_0=1) {
    alphaBoolBlend {blendOption=Linear}
    boneToModify {boneName=dyn_teeth_05}
    constraintSetup [{targetBone {boneName=dyn_jaws}}]
    constraintWeights [1]
   }
   n12 = "Constraint - dyn_teeth_06 (P)"(ComponentPose=@n11, Alpha=1, ConstraintWeights_0=1) {
    alphaBoolBlend {blendOption=Linear}
    boneToModify {boneName=dyn_teeth_06}
    constraintSetup [{targetBone {boneName=dyn_jaws}}]
    constraintWeights [1]
   }
   n13 = RigidBody(ComponentPose=@n12, Alpha=0.35) {
    bClampLinearTranslationLimitToRefPose=true
    bOverrideWorldGravity=true
    overridePhysicsAsset=&/M_MED_Shark_Suit_Teeth_Physics
    simulationSpace=BaseBoneSpace
    alphaBoolBlend {blendOption=Linear}
    baseBoneRef {boneName=neck_02}
    componentLinearAccScale=<0.45 0.45 0.45>
    overrideWorldGravity {z=-500}
   }
   n14 = Anim Dynamics - Bone: dyn_r_legSuit_01(ComponentPose=@n13, Alpha=0.6, ExternalForce="0,0,0") {
    angularDampingOverride=0.9999
    angularSpringConstant=250
    bAngularSpring=true
    bGravityOverrideInSimSpace=true
    bLinearSpring=true
    bOverrideAngularDamping=true
    bOverrideLinearDamping=true
    gravityScale=0.35
    lODThreshold=1
    linearDampingOverride=0.9999
    linearSpringConstant=250
    simulationSpace=BoneRelative
    alphaBoolBlend {blendOption=Linear}
    boundBone {boneName=dyn_r_legSuit_01}
    componentLinearAccScale=<0.25 0.25 0.25>
    physicsBodyDefinitions [{
      sphereCollisionRadius=0
      boundBone {boneName=dyn_r_legSuit_01}
      boxExtents=<2 5 5>
      constraintSetup {
       angularTargetAxis=AxisZ
       angularLimitsMax=<20 20 20>
       angularLimitsMin=<-20 -20 -20>
       angularTarget=<0 0 1>
       linearAxesMax=<2 2 2>
       linearAxesMin=<-2 -2 -2>
      }
      localJointOffset=<0 0 -1>
     }]
    relativeSpaceBone {boneName=root}
   }
   n15 = Anim Dynamics - Bone: dyn_l_legSuit(ComponentPose=@n14, Alpha=0.25, ExternalForce="0,0,0") {
    angularDampingOverride=0.9999
    angularSpringConstant=250
    bAngularSpring=true
    bGravityOverrideInSimSpace=true
    bLinearSpring=true
    bOverrideAngularDamping=true
    bOverrideLinearDamping=true
    gravityScale=0.35
    lODThreshold=1
    linearDampingOverride=0.9999
    linearSpringConstant=250
    simulationSpace=BoneRelative
    alphaBoolBlend {blendOption=Linear}
    boundBone {boneName=dyn_l_legSuit}
    componentLinearAccScale=<0.25 0.25 0.25>
    physicsBodyDefinitions [{
      sphereCollisionRadius=0
      boundBone {boneName=dyn_l_legSuit}
      boxExtents=<2 5 5>
      constraintSetup {
       angularTargetAxis=AxisZ
       angularLimitsMax=<20 20 20>
       angularLimitsMin=<-20 -20 -20>
       angularTarget=<0 0 1>
       linearAxesMax=<2 2 2>
       linearAxesMin=<-2 -2 -2>
      }
      localJointOffset=<0 0 -1>
     }]
    relativeSpaceBone {boneName=root}
   }
   n16 = Anim Dynamics - Bone: dyn_r_legSuit(ComponentPose=@n15, Alpha=0.6, ExternalForce="0,0,0") {
    angularDampingOverride=0.9999
    angularSpringConstant=250
    bAngularSpring=true
    bGravityOverrideInSimSpace=true
    bLinearSpring=true
    bOverrideAngularDamping=true
    bOverrideLinearDamping=true
    gravityScale=0.35
    lODThreshold=1
    linearDampingOverride=0.9999
    linearSpringConstant=250
    simulationSpace=BoneRelative
    alphaBoolBlend {blendOption=Linear}
    boundBone {boneName=dyn_r_legSuit}
    componentLinearAccScale=<0.25 0.25 0.25>
    physicsBodyDefinitions [{
      sphereCollisionRadius=0
      boundBone {boneName=dyn_r_legSuit}
      boxExtents=<2 5 5>
      constraintSetup {
       angularTargetAxis=AxisZ
       angularLimitsMax=<20 20 20>
       angularLimitsMin=<-20 -20 -20>
       angularTarget=<0 0 1>
       linearAxesMax=<2 2 2>
       linearAxesMin=<-2 -2 -2>
      }
      localJointOffset=<0 0 -1>
     }]
    relativeSpaceBone {boneName=root}
   }
   n17 = "Constraint - dyn_l_pouch_offset (PPPPPP)"!(
    ComponentPose=@n16
    Alpha=1
    ConstraintWeights_0=0.1134
    ConstraintWeights_1:float=0.4469
    ConstraintWeights_2:float=0.2735
    ConstraintWeights_3:float=0.1376
    ConstraintWeights_4:float=0.016
    ConstraintWeights_5:float=0.0039
   ) -> (Pose) {
    alphaBoolBlend {blendOption=Linear}
    boneToModify {boneName=dyn_l_pouch_offset}
    constraintSetup [
     {targetBone {boneName=spine_03}}
     {targetBone {boneName=spine_04}}
     {targetBone {boneName=spine_05}}
     {targetBone {boneName=clavicle_l}}
     {targetBone {boneName=upperarm_twist_01_l}}
     {targetBone {boneName=clavicle_r}}
    ]
    constraintWeights [1,1,1,1,1,1]
   }
   n18 = "Constraint - dyn_r_pouch_offset (PPPPPP)"!(
    ComponentPose=@n17
    Alpha=1
    ConstraintWeights_0=0.016
    ConstraintWeights_1:float=0.1134
    ConstraintWeights_2:float=0.4469
    ConstraintWeights_3:float=0.2735
    ConstraintWeights_4:float=0.1376
    ConstraintWeights_5:float=0.0039
   ) -> (Pose) {
    alphaBoolBlend {blendOption=Linear}
    boneToModify {boneName=dyn_r_pouch_offset}
    constraintSetup [
     {targetBone {boneName=upperarm_twist_01_r}}
     {targetBone {boneName=spine_03}}
     {targetBone {boneName=spine_04}}
     {targetBone {boneName=spine_05}}
     {targetBone {boneName=clavicle_r}}
     {targetBone {boneName=clavicle_l}}
    ]
    constraintWeights [1,1,1,1,1,1]
   }
   Output Pose(Result=Component To Local(ComponentPose=@n18)) {layerGroup=None}
  ]}
 EventGraph {nodes [Event BlueprintUpdateAnimation()]}
]
variables []
