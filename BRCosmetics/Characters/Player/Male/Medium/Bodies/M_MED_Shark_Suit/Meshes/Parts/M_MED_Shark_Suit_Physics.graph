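# Physics asset for the Shark Suit male medium body cosmetic, defining collision and joints for its dyn_hair_top bones and head.
# Small identical capsules cover the dyn_hair_top bones, and a larger tilted capsule covers the head.
# Constraints form four chains rooted at the head, one forking into two branches at dyn_hair_top_10.
# All joints share the same moderate swing limits and tight twist limit.
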PhysicsAsset /BRCosmetics/Characters/Player/Male/Medium/Bodies/M_MED_Shark_Suit/Meshes/Parts/M_MED_Shark_Suit_Physics
bodies {
 dyn_hair_top_01 [{shapeType=Capsule, length=3, radius=2, rotation=<rot 90 0 0>}]
 dyn_hair_top_02 [{shapeType=Capsule, length=3, radius=2, rotation=<rot 90 0 0>}]
 dyn_hair_top_03 [{shapeType=Capsule, length=3, radius=2, rotation=<rot 90 0 0>}]
 dyn_hair_top_04 [{shapeType=Capsule, length=3, radius=2, rotation=<rot 90 0 0>}]
 dyn_hair_top_05 [{shapeType=Capsule, length=3, radius=2, rotation=<rot 90 0 0>}]
 dyn_hair_top_06 [{shapeType=Capsule, length=3, radius=2, rotation=<rot 90 0 0>}]
 dyn_hair_top_07 [{shapeType=Capsule, length=3, radius=2, rotation=<rot 90 0 0>}]
 dyn_hair_top_08 [{shapeType=Capsule, length=3, radius=2, rotation=<rot 90 0 0>}]
 dyn_hair_top_10 [{shapeType=Capsule, length=3, radius=2, rotation=<rot 90 0 0>}]
 dyn_hair_top_11 [{shapeType=Capsule, length=3, radius=2, rotation=<rot 90 0 0>}]
 dyn_hair_top_12 [{shapeType=Capsule, length=3, radius=2, rotation=<rot 90 0 0>}]
 dyn_hair_top_13 [{shapeType=Capsule, length=3, radius=2, rotation=<rot 90 0 0>}]
 dyn_hair_top_14 [{shapeType=Capsule, length=3, radius=2, rotation=<rot 90 0 0>}]
 head [{shapeType=Capsule, length=4.2051, radius=6.523, rotation=<rot 90 0 -20.9782>}]
}
constraints [
 {bone1Name=dyn_hair_top_10, bone2Name=head, swing1LimitDegrees=45, swing2LimitDegrees=45, twistLimitDegrees=5}
 {bone1Name=dyn_hair_top_11, bone2Name=dyn_hair_top_10, swing1LimitDegrees=45, swing2LimitDegrees=45, twistLimitDegrees=5}
 {bone1Name=dyn_hair_top_13, bone2Name=dyn_hair_top_11, swing1LimitDegrees=45, swing2LimitDegrees=45, twistLimitDegrees=5}
 {bone1Name=dyn_hair_top_12, bone2Name=dyn_hair_top_10, swing1LimitDegrees=45, swing2LimitDegrees=45, twistLimitDegrees=5}
 {bone1Name=dyn_hair_top_14, bone2Name=dyn_hair_top_12, swing1LimitDegrees=45, swing2LimitDegrees=45, twistLimitDegrees=5}
 {bone1Name=dyn_hair_top_01, bone2Name=head, swing1LimitDegrees=45, swing2LimitDegrees=45, twistLimitDegrees=5}
 {bone1Name=dyn_hair_top_02, bone2Name=dyn_hair_top_01, swing1LimitDegrees=45, swing2LimitDegrees=45, twistLimitDegrees=5}
 {bone1Name=dyn_hair_top_03, bone2Name=dyn_hair_top_02, swing1LimitDegrees=45, swing2LimitDegrees=45, twistLimitDegrees=5}
 {bone1Name=dyn_hair_top_04, bone2Name=head, swing1LimitDegrees=45, swing2LimitDegrees=45, twistLimitDegrees=5}
 {bone1Name=dyn_hair_top_05, bone2Name=dyn_hair_top_04, swing1LimitDegrees=45, swing2LimitDegrees=45, twistLimitDegrees=5}
 {bone1Name=dyn_hair_top_06, bone2Name=dyn_hair_top_05, swing1LimitDegrees=45, swing2LimitDegrees=45, twistLimitDegrees=5}
 {bone1Name=dyn_hair_top_07, bone2Name=head, swing1LimitDegrees=45, swing2LimitDegrees=45, twistLimitDegrees=5}
 {bone1Name=dyn_hair_top_08, bone2Name=dyn_hair_top_07, swing1LimitDegrees=45, swing2LimitDegrees=45, twistLimitDegrees=5}
]
